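# Masked, default-lit decal material for the Numb vehicle cosmetic body.
# The graph holds samples of the decal base color, normal and combined decal textures, plus a Specular scalar parameter, with no connections shown between them.
# Usage is set for skeletal meshes and mesh deformers, while automatic usage flagging and ray-traced shadow casting are disabled.
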Material /VehicleCosmetics/Bodies/Numb/Materials/M_Numb_Decals
graph {
 properties {
  bAutomaticallySetUsageInEditor=false
  bCastRayTracedShadows=false
  bUsedWithMeshDeformer=true
  bUsedWithSkeletalMesh=true
  blendMode=BLEND_Masked
  usedShadingModels=Default Lit
 }
 nodes [
  ScalarParameter() {defaultValue=0.5, parameterName=Specular}
  TextureSample() {samplerType=SAMPLERTYPE_LinearColor, texture=/VehicleCosmetics/Bodies/Numb/Textures/Decals/T_Numb_Decal_BC}
  TextureSample() {samplerType=SAMPLERTYPE_Normal, texture=/VehicleCosmetics/Bodies/Numb/Textures/Decals/T_Numb_Decal_N}
  TextureSample() {samplerType=SAMPLERTYPE_LinearColor, texture=/VehicleCosmetics/Bodies/Numb/Textures/Decals/T_Numb_DecalTextures}
 ]
}
