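# Physics asset for the Numb vehicle body cosmetic, giving the body mesh a simple collision shape.
# It holds a single box on the root bone, longest along X and offset forward and upward from the bone origin.
# It has no other bodies and no constraints, so the vehicle body acts as one rigid volume with no articulated parts.
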PhysicsAsset /VehicleCosmetics/Bodies/Numb/Meshes/Body_Numb_PhysicsDM
bodies {root [{shapeType=Box, extentX=444, extentY=216, extentZ=116, shapeName=root_box, center=<40.5 0 94>}]}
constraints []
